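# Translucent UI material that, by its name, draws the error background in the artists' Asset Validation Tool.
# Nested If nodes threshold the green channel of a square distance-field HUD texture into three levels: zero outside, a full-value border band and a half-value interior.
# A separate branch dims that green channel using a lerp driven by red.
# The listing shows no links from either branch to material outputs.
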Material /Game/Blueprints/Blutilities/ForArtists/AssetValidationTool/Art/AVT_ErrorBG_UIMat
graph {
 properties {blendMode=BLEND_Translucent, materialDomain=MD_UI, pixelDepthOffsetMode=PDOM_Legacy, usedShadingModels=Default Lit}
 nodes [
  Constant() {r=1}
  Constant() {r=1}
  n1 = TextureSample() {texture=/Game/Athena/HUD/Materials/Score/T_UI_Square_DF}
  If(
   A=@n1.G
   B=Constant() {r=0.02}
   "A > B"=If(
    A=TextureSample() {texture=&T_UI_Square_DF}.G
    B=Constant() {r=0.6}
    "A > B"=Constant() {r=0.5}
    "A < B"=Constant() {r=1}
   )
   "A < B"=Constant()
  )
  Multiply(A=LinearInterpolate(Alpha=@n1.R) {constA=1, constB=0.5}, B=@n1.G)
 ]
}
