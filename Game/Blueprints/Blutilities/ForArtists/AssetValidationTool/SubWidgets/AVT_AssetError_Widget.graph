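# Error badge widget for the artists' Asset Validation Tool, displaying one AVT_Error_Struct as a colored label.
# setError stores the struct and fills the title, description, tooltip and border color.
# The tag chip appears when a type tag is set, and the help button only when a help command exists.
# Close and help clicks broadcast dispatchers carrying the widget or help command; removal happens through a separate removeSelf function.
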WidgetBlueprint /Game/Blueprints/Blutilities/ForArtists/AssetValidationTool/SubWidgets/AVT_AssetError_Widget
graphs [
 EventGraph {nodes [
   "On Clicked (Button_71)"() {is=title K2Node_ComponentBoundEvent}
   > Call On Close Button Clicked!(execute:Exec, self:AVT Asset Error Widget Object Reference, widget:AVT Asset Error Widget Object Reference=Self-Reference()) -> (then:Exec) {is=title K2Node_CallDelegate}
   "On Clicked (btnHelp)"() {is=title K2Node_ComponentBoundEvent}
   > Call On Help Button Clicked!(
    execute:Exec
    self:AVT Asset Error Widget Object Reference
    HelpCommand:String=Get!() -> (
     currentError_title_32_#1:String
     currentError_description_35_#2:String
     currentError_tooltip_37_#3:String
     currentError_color_40_#4:Linear Color Structure
     currentError_csvLine_46_#5:String
     currentError_csvHeader_49_#6:String
     currentError_optionalAssetData_52_#7:Asset Data Structure
     currentError_optionalTypeTag_55_#8:String
     currentError_optionalOriginPackageName_62_#9:Name
     currentError_helpCommand_65_#10:String
    ).currentError_helpCommand_65_#10
   ) -> (then:Exec) {is=title K2Node_CallDelegate}
  ]}
 onCloseButtonClicked {nodes [onCloseButtonClicked!() -> (then, widget:AVT Asset Error Widget Object Reference) {is=title K2Node_FunctionEntry}]}
 onHelpButtonClicked {nodes [onHelpButtonClicked!() -> (then, HelpCommand:String) {is=title K2Node_FunctionEntry}]}
 removeSelf {nodes [
   removeSelf() {is=title K2Node_FunctionEntry}
   > RemoveFromParent()
  ]}
 setError {nodes [
   n1 = setError!() -> (then, error:AVT Error Struct Structure) {is=title K2Node_FunctionEntry}
   > Set currentError(currentError:AVT Error Struct Structure=@n1) -> (Output_Get:AVT Error Struct Structure)
   > n2 = Sequence()
   n3 = Break AVT Error Struct!(AVT_Error_Struct:"AVT Error Struct Structure (by ref)"=Get currentError() -> (currentError:AVT Error Struct Structure)) -> (
    title_32_#1:String
    description_35_#2:String
    tooltip_37_#3:String
    color_40_#4:Linear Color Structure
    csvLine_46_#5:String
    csvHeader_49_#6:String
    optionalAssetData_52_#7:Asset Data Structure
    optionalTypeTag_55_#8:String
    optionalOriginPackageName_62_#9:Name
    helpCommand_65_#10:String
   ) {is=title K2Node_BreakStruct}
   > "SetText (Text)"(self=Get txtType() -> (txtType:Common Text Object Reference), InText="To Text (String)"(InString=@n3))
   > "SetText (Text)"(
    self=Get txtSubType() -> (txtSubType:Common Text Object Reference)
    InText="To Text (String)"(InString=@n3.description_35_#2)
   )
   n4 = Get currentError()
   > SetToolTipText!(
    execute
    self:Widget Object Reference
    InToolTipText:"Text (by ref)"="To Text (String)"(InString=Break AVT Error Struct!(AVT_Error_Struct=@n4) -> (
      tooltip_37_#3
      optionalAssetData_52_#7
      optionalTypeTag_55_#8
      optionalOriginPackageName_62_#9
      helpCommand_65_#10
     ))
   ) -> (then)
   n5 = Break AVT Error Struct!(AVT_Error_Struct=Get currentError()) -> (
    color_40_#4:Linear Color Structure
    optionalAssetData_52_#7
    optionalTypeTag_55_#8
    optionalOriginPackageName_62_#9
    helpCommand_65_#10
   )
   > SetBrushColor(self=Get Border_8() -> (Border_8:Border Object Reference), InBrushColor=@n5)
   > Branch(Condition="Not Equal Exactly (String)"(A=@n5.optionalTypeTag_55_#8))
   > SetVisibility(self=Get borderTag() -> (borderTag:Border Object Reference), InVisibility=HitTestInvisible)
   > "SetText (Text)"(
    self=Get txtTag() -> (txtTag:Common Text Object Reference)
    InText="To Text (String)"(InString=@n5.optionalTypeTag_55_#8)
   )
   n6 = Break AVT Error Struct!(AVT_Error_Struct=Get currentError()) -> (helpCommand_65_#10)
   SetVisibility(
    execute=@n2.then_1
    self=Get btnHelp() -> (btnHelp:Button Object Reference)
    InVisibility=Select(Option 0=Collapsed, Option 1=Visible, Index="Not Equal Exactly (String)"(A=@n6))
   )
   n7 = Break AVT Error Struct!(AVT_Error_Struct=Get currentError()) -> (
    title_32_#1:String
    description_35_#2:String
    tooltip_37_#3:String
    color_40_#4:Linear Color Structure
    csvLine_46_#5:String
    csvHeader_49_#6:String
    optionalAssetData_52_#7:Asset Data Structure
    optionalTypeTag_55_#8:String
    optionalOriginPackageName_62_#9:Name
    helpCommand_65_#10
   )
   > SetToolTipText(self=Get btnHelp(), InToolTipText="To Text (String)"(InString=Append(A="Help for ", B=@n7)))
  ]}
]
variables [
 currentError {
  type=AVT_Error_Struct
  default {
   csvHeader=""
   csvLine=""
   description=""
   helpCommand=""
   optionalOriginPackageName=None
   optionalTypeTag=""
   title=""
   tooltip=""
   color=<rgba 0 0 0 0>
   optionalAssetData {}
  }
 }
 onCloseButtonClicked {}
 onHelpButtonClicked {}
]
widgets {
 info {inheritedWidgetCount=0, parentClass=&UserWidget, rootWidgetClass=&CanvasPanel, widgetCount=12}
 widgets [
  CanvasPanel_0 {widgetClassPath=&CanvasPanel, widget=CanvasPanel_0}
  HorizontalBox_0 {
   widgetClassPath=&HorizontalBox
   parent=CanvasPanel_0
   widget=HorizontalBox_0
   slot {
    class=CanvasPanelSlot
    properties {bAutoSize=true, layoutData {alignment=<0 0>, anchors {maximum=<0 0>, minimum=<0 0>}, offsets=<ltrb 0 0 181.6667 31.5315>}}
   }
  }
  Border_8 {
   widgetClassPath=&Border
   bIsVariable=true
   parent=HorizontalBox_0
   widget=Border_8
   properties {
    verticalAlignment=VAlign_Center
    background {
     bIsDynamicallyLoaded=false
     drawAs=Box
     imageType=NoImage
     mirroring=NoMirror
     resourceName=None
     resourceObject=/Game/Blueprints/Blutilities/ForArtists/AssetValidationTool/Art/AVT_ErrorBG_UIMat
     tiling=NoTile
     imageSize=<20 20>
     margin=<ltrb 0.15 0.15 0.15 0.15>
     outlineSettings {
      bUseBrushTransparency=false
      roundingType=HalfHeightRadius
      width=0
      color {colorUseRule=UseColor_Specified, specifiedColor=<rgba 0 0 0 0>}
      cornerRadii=<0 0 0 0>
     }
     tintColor {colorUseRule=UseColor_Specified, specifiedColor=<rgba 1 1 1 1>}
     uVRegion {bIsValid=false, max=<0 0>, min=<0 0>}
    }
    brushColor=<rgba 0.5781 0.3232 0.0514 1>
    padding=<ltrb 3 1 3 1>
   }
   slot {class=HorizontalBoxSlot}
  }
  VerticalBox_5 {
   widgetClassPath=&VerticalBox
   parent=Border_8
   widget=VerticalBox_5
   properties {clipping=ClipToBounds}
   slot {class=BorderSlot, properties {verticalAlignment=VAlign_Center, padding=<ltrb 3 1 3 1>}}
  }
  txtType {
   widgetClassPath=&CommonTextBlock
   bIsVariable=true
   parent=VerticalBox_5
   widget=txtType
   properties {
    justification=Center
    text=COL
    textTransformPolicy=ToUpper
    colorAndOpacity {colorUseRule=UseColor_Specified, specifiedColor=<rgba 0.9323 0.9323 0.9323 1>}
    font {
     bForceMonospaced=false
     bMaterialIsStencil=false
     fontMaterial=null
     fontObject=&NotoSans
     letterSpacing=0
     monospacedWidth=1
     size=8
     skewAmount=0
     typefaceFontName=NotoSans-Bold
     outlineSettings {
      bApplyOutlineToDropShadows=false
      bMiteredCorners=false
      bSeparateFillAlpha=false
      outlineMaterial=null
      outlineSize=0
      outlineColor=<rgba 0 0 0 1>
     }
    }
    shadowOffset=<0 0>
   }
   slot {class=VerticalBoxSlot}
  }
  txtSubType {
   widgetClassPath=&CommonTextBlock
   bIsVariable=true
   parent=VerticalBox_5
   widget=txtSubType
   properties {
    justification=Center
    text=NoLOD
    colorAndOpacity {colorUseRule=UseColor_Specified, specifiedColor=<rgba 0.9323 0.9323 0.9323 1>}
    font {
     bForceMonospaced=false
     bMaterialIsStencil=false
     fontMaterial=null
     fontObject=&NotoSans
     letterSpacing=0
     monospacedWidth=1
     size=8
     skewAmount=0
     typefaceFontName=NotoSans-Bold
     outlineSettings {
      bApplyOutlineToDropShadows=false
      bMiteredCorners=false
      bSeparateFillAlpha=false
      outlineMaterial=null
      outlineSize=0
      outlineColor=<rgba 0 0 0 1>
     }
    }
    shadowOffset=<0 0>
   }
   slot {class=VerticalBoxSlot}
  }
  Spacer_83 {widgetClassPath=&Spacer, parent=HorizontalBox_0, widget=Spacer_83, properties {size=<10 1>}, slot {class=HorizontalBoxSlot}}
  borderTag {
   widgetClassPath=&Border
   bIsVariable=true
   parent=CanvasPanel_0
   widget=borderTag
   properties {
    visibility=Hidden
    background {
     bIsDynamicallyLoaded=false
     drawAs=RoundedBox
     imageType=NoImage
     mirroring=NoMirror
     resourceName=None
     resourceObject=null
     tiling=NoTile
     imageSize=<1 1>
     margin=<ltrb 0 0 0 0>
     outlineSettings {
      bUseBrushTransparency=false
      roundingType=HalfHeightRadius
      width=0
      color {colorUseRule=UseColor_Specified, specifiedColor=<rgba 0 0 0 0>}
      cornerRadii=<0 0 0 0>
     }
     tintColor {colorUseRule=UseColor_Specified, specifiedColor=<rgba 1 1 1 1>}
     uVRegion {bIsValid=false, max=<0 0>, min=<0 0>}
    }
    brushColor=<rgba 0.7292 0.6539 0.3255 1>
    padding=<ltrb 2 0 2 0>
    renderTransform {angle=0, scale=<1 1>, shear=<0 0>, translation=<-4 -2>}
   }
   slot {
    class=CanvasPanelSlot
    properties {bAutoSize=true, layoutData {alignment=<0 0>, anchors {maximum=<0 0>, minimum=<0 0>}, offsets=<ltrb 0 0 35.1852 20.5531>}}
   }
  }
  txtTag {
   widgetClassPath=&CommonTextBlock
   bIsVariable=true
   parent=borderTag
   widget=txtTag
   properties {
    justification=Center
    text=FX
    textTransformPolicy=ToUpper
    colorAndOpacity {colorUseRule=UseColor_Specified, specifiedColor=<rgba 0.0156 0.0156 0.0156 1>}
    font {
     bForceMonospaced=false
     bMaterialIsStencil=false
     fontMaterial=null
     fontObject=&NotoSans
     letterSpacing=0
     monospacedWidth=1
     size=7
     skewAmount=0
     typefaceFontName=NotoSans-Bold
     outlineSettings {
      bApplyOutlineToDropShadows=false
      bMiteredCorners=false
      bSeparateFillAlpha=false
      outlineMaterial=null
      outlineSize=0
      outlineColor=<rgba 0 0 0 1>
     }
    }
    shadowOffset=<0 0>
   }
   slot {class=BorderSlot, properties {padding=<ltrb 2 0 2 0>}}
  }
  Button_71 {
   widgetClassPath=&Button
   parent=CanvasPanel_0
   widget=Button_71
   properties {widgetStyle {
     clickedSlateSound {resourceObject=null}
     disabled {
      bIsDynamicallyLoaded=false
      drawAs=NoDrawType
      imageType=NoImage
      mirroring=NoMirror
      resourceName=None
      resourceObject=null
      tiling=NoTile
      imageSize=<0 0>
      margin=<ltrb 0 0 0 0>
      outlineSettings {
       bUseBrushTransparency=false
       roundingType=HalfHeightRadius
       width=0
       color {colorUseRule=UseColor_Specified, specifiedColor=<rgba 0 0 0 0>}
       cornerRadii=<0 0 0 0>
      }
      tintColor {colorUseRule=UseColor_Specified, specifiedColor=<rgba 1 1 1 1>}
      uVRegion {bIsValid=false, max=<0 0>, min=<0 0>}
     }
     disabledForeground {colorUseRule=UseColor_Specified, specifiedColor=<rgba 0.5271 0.5271 0.5271 1>}
     hovered {
      bIsDynamicallyLoaded=false
      drawAs=RoundedBox
      imageType=NoImage
      mirroring=NoMirror
      resourceName=None
      resourceObject=/Game/Blueprints/Blutilities/ForTechArt/ArtAssets/Textures/Checkbox/TA_HierarchyShowHide_Checked_Over
      tiling=NoTile
      imageSize=<32 32>
      margin=<ltrb 0 0 0 0>
      outlineSettings {
       bUseBrushTransparency=true
       roundingType=FixedRadius
       width=1
       color {colorUseRule=UseColor_Specified, specifiedColor=<rgba 0.6951 0.6951 0.6951 1>}
       cornerRadii=<4 4 4 4>
      }
      tintColor {colorUseRule=UseColor_Specified, specifiedColor=<rgba 1 0.2292 0.2292 1>}
      uVRegion {bIsValid=false, max=<0 0>, min=<0 0>}
     }
     hoveredForeground {colorUseRule=UseColor_Specified, specifiedColor=<rgba 0 1 1 1>}
     hoveredSlateSound {resourceObject=null}
     normal {
      bIsDynamicallyLoaded=false
      drawAs=RoundedBox
      imageType=NoImage
      mirroring=NoMirror
      resourceName=None
      resourceObject=&TA_HierarchyShowHide_Checked_Over
      tiling=NoTile
      imageSize=<32 32>
      margin=<ltrb 0 0 0 0>
      outlineSettings {
       bUseBrushTransparency=true
       roundingType=FixedRadius
       width=1
       color {colorUseRule=UseColor_Specified, specifiedColor=<rgba 0.6951 0.6951 0.6951 1>}
       cornerRadii=<4 4 4 4>
      }
      tintColor {colorUseRule=UseColor_Specified, specifiedColor=<rgba 0.6927 0.0505 0.0505 1>}
      uVRegion {bIsValid=false, max=<0 0>, min=<0 0>}
     }
     normalForeground {colorUseRule=UseColor_Specified, specifiedColor=<rgba 0 1 1 1>}
     normalPadding=<ltrb 12 1.5 12 1.5>
     pressed {
      bIsDynamicallyLoaded=false
      drawAs=RoundedBox
      imageType=NoImage
      mirroring=NoMirror
      resourceName=None
      resourceObject=&TA_HierarchyShowHide_Checked_Over
      tiling=NoTile
      imageSize=<32 32>
      margin=<ltrb 0 0 0 0>
      outlineSettings {
       bUseBrushTransparency=true
       roundingType=FixedRadius
       width=1
       color {colorUseRule=UseColor_Specified, specifiedColor=<rgba 0.6951 0.6951 0.6951 1>}
       cornerRadii=<4 4 4 4>
      }
      tintColor {colorUseRule=UseColor_Specified, specifiedColor=<rgba 1 0.0729 0.0729 1>}
      uVRegion {bIsValid=false, max=<0 0>, min=<0 0>}
     }
     pressedForeground {colorUseRule=UseColor_Specified, specifiedColor=<rgba 0 1 1 1>}
     pressedPadding=<ltrb 12 1.5 12 1.5>
     pressedSlateSound {resourceObject=null}
    }}
   slot {class=CanvasPanelSlot, properties {layoutData {alignment=<0 0>, anchors {maximum=<1 0>, minimum=<1 0>}, offsets=<ltrb -15 -5 14 14>}}}
  }
  btnHelp {
   widgetClassPath=&Button
   parent=CanvasPanel_0
   widget=btnHelp
   properties {
    toolTipText=HELP
    renderTransform {angle=0, scale=<1 1>, shear=<0 0>, translation=<0 3>}
    widgetStyle {
     clickedSlateSound {resourceObject=null}
     disabled {
      bIsDynamicallyLoaded=false
      drawAs=Box
      imageType=NoImage
      mirroring=NoMirror
      resourceName=None
      resourceObject=/Engine/EditorResources/S_ReflActorIcon
      tiling=NoTile
      imageSize=<256 256>
      margin=<ltrb 0 0 0 0>
      outlineSettings {
       bUseBrushTransparency=true
       roundingType=FixedRadius
       width=1
       color {colorUseRule=UseColor_Specified, specifiedColor=<rgba 0.6951 0.6951 0.6951 1>}
       cornerRadii=<4 4 4 4>
      }
      tintColor {colorUseRule=UseColor_Specified, specifiedColor=<rgba 0.4955 0.4955 0.4955 1>}
      uVRegion {bIsValid=false, max=<0 0>, min=<0 0>}
     }
     disabledForeground {colorUseRule=UseColor_Specified, specifiedColor=<rgba 0.5271 0.5271 0.5271 1>}
     hovered {
      bIsDynamicallyLoaded=false
      drawAs=Box
      imageType=NoImage
      mirroring=NoMirror
      resourceName=None
      resourceObject=&S_ReflActorIcon
      tiling=NoTile
      imageSize=<256 256>
      margin=<ltrb 0 0 0 0>
      outlineSettings {
       bUseBrushTransparency=true
       roundingType=FixedRadius
       width=1
       color {colorUseRule=UseColor_Specified, specifiedColor=<rgba 0.6951 0.6951 0.6951 1>}
       cornerRadii=<4 4 4 4>
      }
      tintColor {colorUseRule=UseColor_Specified, specifiedColor=<rgba 0.8698 0.8048 0.4382 1>}
      uVRegion {bIsValid=false, max=<0 0>, min=<0 0>}
     }
     hoveredForeground {colorUseRule=UseColor_Specified, specifiedColor=<rgba 0 1 1 1>}
     hoveredSlateSound {resourceObject=null}
     normal {
      bIsDynamicallyLoaded=false
      drawAs=Box
      imageType=NoImage
      mirroring=NoMirror
      resourceName=None
      resourceObject=&S_ReflActorIcon
      tiling=NoTile
      imageSize=<256 256>
      margin=<ltrb 0 0 0 0>
      outlineSettings {
       bUseBrushTransparency=true
       roundingType=FixedRadius
       width=1
       color {colorUseRule=UseColor_Specified, specifiedColor=<rgba 0.6951 0.6951 0.6951 1>}
       cornerRadii=<4 4 4 4>
      }
      tintColor {colorUseRule=UseColor_Specified, specifiedColor=<rgba 0.4955 0.4955 0.4955 1>}
      uVRegion {bIsValid=false, max=<0 0>, min=<0 0>}
     }
     normalForeground {colorUseRule=UseColor_Specified, specifiedColor=<rgba 0 1 1 1>}
     normalPadding=<ltrb 0 0 0 0>
     pressed {
      bIsDynamicallyLoaded=false
      drawAs=Box
      imageType=NoImage
      mirroring=NoMirror
      resourceName=None
      resourceObject=&S_ReflActorIcon
      tiling=NoTile
      imageSize=<256 256>
      margin=<ltrb 0 0 0 0>
      outlineSettings {
       bUseBrushTransparency=true
       roundingType=FixedRadius
       width=1
       color {colorUseRule=UseColor_Specified, specifiedColor=<rgba 0.6951 0.6951 0.6951 1>}
       cornerRadii=<4 4 4 4>
      }
      tintColor {colorUseRule=UseColor_Specified, specifiedColor=<rgba 0.4955 0.4955 0.4955 1>}
      uVRegion {bIsValid=false, max=<0 0>, min=<0 0>}
     }
     pressedForeground {colorUseRule=UseColor_Specified, specifiedColor=<rgba 0 1 1 1>}
     pressedPadding=<ltrb 0 0 0 0>
     pressedSlateSound {resourceObject=null}
    }
   }
   slot {
    class=CanvasPanelSlot
    properties {bAutoSize=true, zOrder=4, layoutData {alignment=<0 0>, anchors {maximum=<1 1>, minimum=<1 1>}, offsets=<ltrb -15 -15 -10.73 11>}}
   }
  }
  Image_1 {
   widgetClassPath=&Image
   bIsVariable=true
   parent=btnHelp
   widget=Image_1
   properties {brush {
     bIsDynamicallyLoaded=false
     drawAs=Image
     imageType=NoImage
     mirroring=NoMirror
     resourceName=None
     resourceObject=/Game/Athena/UI/InGame/Creative/PropertyEditor/Icon_Picker/Textures/64x64/T_UI_IconLibrary_questionmark_64
     tiling=NoTile
     imageSize=<14 14>
     margin=<ltrb 0 0 0 0>
     outlineSettings {
      bUseBrushTransparency=false
      roundingType=HalfHeightRadius
      width=0
      color {colorUseRule=UseColor_Specified, specifiedColor=<rgba 0 0 0 0>}
      cornerRadii=<0 0 0 0>
     }
     tintColor {colorUseRule=UseColor_Specified, specifiedColor=<rgba 1 0.7188 0.1571 1>}
     uVRegion {bIsValid=false, max=<0 0>, min=<0 0>}
    }}
   slot {class=ButtonSlot, properties {horizontalAlignment=HAlign_Fill, verticalAlignment=VAlign_Fill, padding=<ltrb 2 2 2 2>}}
  }
 ]
}
