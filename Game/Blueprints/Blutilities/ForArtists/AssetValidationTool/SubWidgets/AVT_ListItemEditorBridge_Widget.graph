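# Editor utility widget that serves as the Asset Validation Tool's list-item bridge to Content Browser syncing and level selection.
# syncBrowserToAsset highlights a package's asset in the Content Browser.
# selectMeshInLevel resolves a package via the Asset Registry, compares component meshes against it, then selects owning actors.
# Only StaticMeshComponents are gathered, so the skeletal-mesh fallback receives none; both wrapper functions always return success false.
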EditorUtilityWidgetBlueprint /Game/Blueprints/Blutilities/ForArtists/AssetValidationTool/SubWidgets/AVT_ListItemEditorBridge_Widget
graphs [
 EventGraph {nodes [
   Event PreConstruct()
   Event Construct()
   Event Tick()
  ]}
 SyncAssetInCB {nodes [
   n1 = SyncAssetInCB!() -> (then, packageName:String) {is=title K2Node_FunctionEntry}
   > SyncBrowserToObjects(AssetPaths=Make Array("[0]":String=@n1) -> (Array:Array of Strings))
  ]}
 findMeshUseInWorld {nodes [
   n1 = findMeshUseInWorld!() -> (then, meshPackageName:String) {is=title K2Node_FunctionEntry}
   n2 = String To Name(InString=@n1)
   > n3 = GetAssets!(
    execute
    self:Asset Registry Interface=GetAssetRegistry()
    Filter:"ARFilter Structure (by ref)"=MakeARFilter!(
     PackageNames:Array of Names=Make Array("[0]":Name=@n2) -> (Array:Array of Names)
     PackagePaths:Array of Names
     SoftObjectPaths:Array of Soft Object Path Structures
     ClassPaths:Array of Top Level Asset Path Structures
     RecursiveClassPathsExclusionSet:Set of Top Level Asset Path Structures
     ClassNames:Array of Names=Make Array!("[0]"=StaticMesh, "[1]":Name=SkeletalMesh) -> (Array)
     RecursiveClassesExclusionSet:Set of Names
     bRecursivePaths:bool=false
     bRecursiveClasses:bool=false
     bIncludeOnlyOnDiskAssets:bool=false
    ) -> (ReturnValue:ARFilter Structure)
    bSkipARFilteredAssets:bool=true
   ) -> (then, OutAssetData:Array of Asset Data Structures, ReturnValue:bool)
   > n4 = For Each Loop(Array:Array of Asset Data Structures=@n3) -> (Array Element:Asset Data Structure)
   SetSelectedLevelActors(
    execute=@n4.Completed
    self=Get EditorActorSubsystem()
    ActorsToSelect=Get l_finalActors() -> (l_finalActors:Array of Actor Object References)
   )
   n5 = Sequence(execute=@n4.LoopBody)
   > n6 = GetAllLevelActorsComponents!(execute, self:Editor Actor Subsystem Object Reference=Get EditorActorSubsystem()) -> (then, ReturnValue:Array of Actor Component Object References)
   > n7 = Filter by Class!(
    execute
    TargetArray:Array of Object References=@n6
    ObjectClass:Object Class Reference=&StaticMeshComponent
    FilterType:EEditorScriptingFilterType Enum
   ) -> (then, ReturnValue:Array of Static Mesh Component Object References)
   > Append Array!(
    execute
    TargetArray:Array of Actor Component Object References=Get l_components() -> (l_components:Array of Actor Component Object References)
    SourceArray:Array of Actor Component Object References=@n7
   ) -> (then)
   n8 = For Each Loop(Exec=@n5.then_1, Array:Array of Actor Component Object References=Get l_components()) -> (Array Element:Actor Component Object Reference)
   > n9 = Cast To StaticMeshComponent(Object=@n8)
   > n10 = Set l_object(l_object:Object Reference=Get StaticMesh(self=@n9)) -> (Output_Get:Object Reference)
   Branch(
    execute=[@n12.then, @n10.then]
    Condition="Equal (Object)"(A=Get l_object() -> (l_object:Object Reference), B=GetAsset(InAssetData=@n4)) {is=title K2Node_CallFunction}
   )
   > Add(TargetArray:Array of Actor Object References=Get l_finalActors(), NewItem:"Actor Object Reference (by ref)"=GetOwner(self=@n8)) {is=title K2Node_CallArrayFunction}
   n11 = Cast To SkeletalMeshComponent!(execute=@n9.CastFailed, Object=@n8) -> (then, CastFailed, AsSkeletal Mesh Component:Skeletal Mesh Component Object Reference) {is=title K2Node_DynamicCast}
   > n12 = Set l_object(l_object=Get SkeletalMesh!(self:Skinned Mesh Component Object Reference=@n11) -> (SkeletalMesh:Skeletal Mesh Object Reference))
  ]}
 selectMeshInLevel {nodes [
   n1 = selectMeshInLevel!() -> (then, mesh:String) {is=title K2Node_FunctionEntry}
   > findMeshUseInWorld!(execute, self, meshPackageName:String=@n1) -> (then) {is=title K2Node_CallFunction}
   > Return Node!(execute, success:bool=false)
  ]}
 syncBrowserToAsset {nodes [
   n1 = syncBrowserToAsset!() -> (then, asset:String) {is=title K2Node_FunctionEntry}
   > SyncAssetInCB!(execute, self, packageName:String=@n1) -> (then) {is=title K2Node_CallFunction}
   > Return Node(success=false)
  ]}
]
variables []
widgets {
 info {rootWidgetClass=&CanvasPanel, widgetCount=1}
 widgets [CanvasPanel_0 {widgetClassPath=&CanvasPanel, widget=CanvasPanel_0}]
}
