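# List row widget for the artist-facing Asset Validation Tool, showing one asset's name and its validation errors.
# On data update it sets the label from AVT_ItemData and spawns an error widget per error, re-dispatching their close and help clicks.
# Find buttons use an editor bridge widget to sync the Content Browser or select the mesh in the level.
# The remove button removes the row from its parent TA_Generic01_List, and an edge line reflects selection.
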WidgetBlueprint /Game/Blueprints/Blutilities/ForArtists/AssetValidationTool/SubWidgets/AVT_Asset_ListItemWidget
graphs [
 EventGraph {nodes [
   "On Clicked (btnRemove)"!() -> (OutputDelegate, then, widget:User Widget Object Reference, event:Pointer Event Structure) {is=title K2Node_ComponentBoundEvent}
   > n1 = FindParentWidgetOfType!(
    execute
    StartingWidget:Widget Object Reference=Self-Reference()
    Type:Widget Class Reference=/Game/Blueprints/Blutilities/ForTechArt/Widgets/List/TA_Generic01_List.TA_Generic01_List_C
   ) -> (then, ReturnValue:TA Generic 01 List Object Reference)
   > n2 = Cast To TA_Generic01_List!(execute, Object=@n1) -> (then, CastFailed, AsTA Generic 01 List:TA Generic 01 List Object Reference) {is=title K2Node_DynamicCast}
   > removeItems!(
    execute
    self:TA Generic 01 List Object Reference=@n2
    itemWidgets:Array of User Widget Object References=Make Array("[0]":User Widget Object Reference=Self-Reference()) -> (Array:Array of User Widget Object References)
   ) -> (then)
   "On Clicked (btnFindCB)"() {is=title K2Node_ComponentBoundEvent}
   > n3 = Cast To AVT_ItemData!(execute, Object=Get currentListItemData() -> (currentListItemData:Object Reference)) -> (then, CastFailed, AsAVT Item Data:AVT Item Data Object Reference) {is=title K2Node_DynamicCast}
   > syncBrowserToAsset!(
    execute:Exec
    self:Object Reference=Get AVT_ListItemEditorBridge_Widget() -> (AVT_ListItemEditorBridge_Widget:AVT List Item Editor Bridge Widget Object Reference)
    asset:String="Get (a copy)"(Array=Get strings!(self:TA Generic List Item Data Object Reference=@n3) -> (strings:Array of Strings))
   ) -> (then:Exec, success:bool) {is=title K2Node_Message}
   "On Clicked (btnFindWorld)"() {is=title K2Node_ComponentBoundEvent}
   > n4 = Cast To AVT_ItemData(Object=Get currentListItemData())
   > selectMeshInLevel!(execute:Exec, self:Object Reference=Get AVT_ListItemEditorBridge_Widget(), mesh:String="Get (a copy)"(Array=Get strings(self=@n4))) -> (then:Exec, success:bool) {is=title K2Node_Message}
   n5 = Event updateAfterDataChange!() -> (OutputDelegate, then, currentListItemData:Object Reference) {is=title K2Node_Event}
   > Parent: updateAfterDataChange!(execute, currentListItemData:Object Reference=@n5.currentListItemData) -> (then) {is=title K2Node_CallParentFunction}
   > n6 = Cast To AVT_ItemData(Object=@n5.currentListItemData)
   n7 = "To Text (String)"(InString=Get displayString!(self:TA Generic List Item Data Object Reference=@n6) -> (displayString:String))
   > "SetText (Text)"(self=Get txtMain() -> (txtMain:Text Object Reference), InText=@n7)
   > ClearChildren(self=Get wrapBoxErrors() -> (wrapBoxErrors:Wrap Box Object Reference))
   > populateErrors!(execute, self, itemData:AVT Item Data Object Reference=@n6) -> (then)
   n8 = "To String (Text)"(InText=@n7)
   > SetToolTipText!(
    execute
    self:Widget Object Reference=Get btnRemove() -> (btnRemove:TA Generic 01 Button Object Reference)
    InToolTipText:"Text (by ref)"="To Text (String)"(InString=Append(A="Remove ", B=@n8))
   ) -> (then)
  ]}
 getBackgroundWidget {nodes [
   getBackgroundWidget() {is=title K2Node_FunctionEntry}
   > Return Node!(execute, backgroundWidget:Widget Object Reference=Get borderMain() -> (borderMain:Border Object Reference))
  ]}
 getLabelWidget {nodes [
   getLabelWidget() {is=title K2Node_FunctionEntry}
   > Return Node!(execute, labelWidget:Widget Object Reference=Get txtMain())
  ]}
 handleErrorWidgetCloseClicked {nodes [
   n1 = handleErrorWidgetCloseClicked!() -> (then, widget:AVT Asset Error Widget Object Reference) {is=title K2Node_FunctionEntry}
   > n2 = Cast To AVT_ItemData(Object=Get currentListItemData())
   > Call On Error Close Clicked!(
    execute:Exec
    self:AVT Asset List Item Widget Object Reference
    errorWidget:AVT Asset Error Widget Object Reference=@n1
    itemData:AVT Item Data Object Reference=@n2
   ) -> (then:Exec) {is=title K2Node_CallDelegate}
  ]}
 handleErrorWidgetHelpClicked {nodes [
   n1 = handleErrorWidgetHelpClicked!() -> (then, HelpCommand:String) {is=title K2Node_FunctionEntry}
   > Call On Error Help Clicked!(execute:Exec, self:AVT Asset List Item Widget Object Reference, helpCommmand:String=@n1) -> (then:Exec) {is=title K2Node_CallDelegate}
  ]}
 onErrorCloseClicked {nodes [onErrorCloseClicked!() -> (then, errorWidget:AVT Asset Error Widget Object Reference, itemData:AVT Item Data Object Reference) {is=title K2Node_FunctionEntry}]}
 onErrorHelpClicked {nodes [onErrorHelpClicked!() -> (then, helpCommmand:String) {is=title K2Node_FunctionEntry}]}
 populateErrors {nodes [
   n1 = populateErrors!() -> (then, itemData:AVT Item Data Object Reference) {is=title K2Node_FunctionEntry}
   > n2 = For Each Loop(Array:Array of AVT Error Struct Structures=Break AVT Asset Data Struct!(AVT_AssetData_Struct:"AVT Asset Data Struct Structure (by ref)"=Get assetErrorData!(self:AVT Item Data Object Reference=@n1) -> (assetErrorData:AVT Asset Data Struct Structure)) -> (
     asset_39_#1:Object Reference
     errors_35_#2:Array of AVT Error Struct Structures
     csvLine_42_#3:String
     assetPackageName_46_#4:String
    ) {is=title K2Node_BreakStruct}.errors_35_#2) -> (Array Element:AVT Error Struct Structure)
   > n3 = Create Widget!(
    execute:Exec
    Class:User Widget Class Reference=&/AVT_AssetError_Widget.AVT_AssetError_Widget_C
    OwningPlayer:Player Controller Object Reference
   ) -> (then:Exec, ReturnValue:AVT Asset Error Widget Object Reference)
   > setError!(execute, self:AVT Asset Error Widget Object Reference=@n3, error:AVT Error Struct Structure=@n2) -> (then)
   > AddChildToWrapBox!(execute, self:Wrap Box Object Reference=Get wrapBoxErrors(), Content:Widget Object Reference=@n3) -> (then, ReturnValue:Wrap Box Slot Object Reference)
   > Bind Event to On Close Button Clicked(self:AVT Asset Error Widget Object Reference=@n3, Delegate=Create Event()) {is=title K2Node_AddDelegate}
   > Bind Event to On Help Button Clicked(self:AVT Asset Error Widget Object Reference=@n3, Delegate=Create Event()) {is=title K2Node_AddDelegate}
  ]}
 setTextAlignment {nodes [
   n1 = setTextAlignment!() -> (then, alignment:EHorizontalAlignment Enum=HAlign_Fill) {is=title K2Node_FunctionEntry}
   > n2 = Cast To ScaleBoxSlot!(execute, Object=Get Slot!(self:Widget Object Reference=Get txtMain()) -> (Slot:Panel Slot Object Reference)) -> (then, CastFailed, AsScale Box Slot:Scale Box Slot Object Reference) {is=title K2Node_DynamicCast}
   > SetHorizontalAlignment(self:Scale Box Slot Object Reference=@n2, InHorizontalAlignment=@n1)
  ]}
 validateBackgroundColor {nodes [
   validateBackgroundColor() {is=title K2Node_FunctionEntry}
   > SetColorAndOpacity(
    self:Image Object Reference=Get imageEdge() -> (imageEdge:Image Object Reference)
    InColorAndOpacity:Linear Color Structure=SelectColor(A=(R=0.0284,G=0.3763,B=1,A=1), B=(R=0.8123,G=0.8217,B=0.8385,A=1), bPickA=getSelected!() -> (isSelected:bool))
   )
   > n1 = Parent: validateBackgroundColor!(execute) -> (then, handled:bool) {is=title K2Node_CallParentFunction}
   > Return Node!(execute, handled:bool=@n1)
  ]}
 validateVisuals {nodes [
   validateVisuals() {is=title K2Node_FunctionEntry}
   n1 = getSelected()
   > SetBrushTintColor!(
    execute
    self:Image Object Reference=Get imageEdge()
    TintColor:Slate Color Structure=Make Slate Color(SpecifiedColor=Select(Option 0:Linear Color Structure=(R=1,G=1,B=1,A=0.5), Option 1:Linear Color Structure=(R=1,G=1,B=1,A=1), Index=@n1) -> (ReturnValue:Linear Color Structure))
   ) -> (then)
   > n2 = Parent: validateVisuals!(execute) -> (then, " ":bool) {is=title K2Node_CallParentFunction}
   > Return Node!(execute, " ":bool=@n2)
  ]}
]
variables [
 onErrorCloseClicked {}
 onErrorHelpClicked {}
]
widgets {
 info {
  inheritedWidgetCount=0
  parentClass=/Game/Blueprints/Blutilities/ForTechArt/Widgets/Buttons/TA_BaseEditorButtonWidget.TA_BaseEditorButtonWidget_C
  rootWidgetClass=&Overlay
  widgetCount=21
 }
 widgets [
  Overlay_7 {widgetClassPath=&Overlay, widget=Overlay_7}
  VerticalBox_0 {
   widgetClassPath=&VerticalBox
   parent=Overlay_7
   widget=VerticalBox_0
   slot {class=OverlaySlot, properties {horizontalAlignment=HAlign_Fill}}
  }
  borderMain {
   widgetClassPath=&Border
   bIsVariable=true
   parent=VerticalBox_0
   widget=borderMain
   properties {
    background {
     bIsDynamicallyLoaded=false
     drawAs=Box
     imageType=NoImage
     mirroring=NoMirror
     resourceName=None
     resourceObject=/Game/Blueprints/Blutilities/ForArtists/AssetValidationTool/Art/AVT_ErrorBG_UIMat
     tiling=NoTile
     imageSize=<16 16>
     margin=<ltrb 0.15 0.15 0.15 0.15>
     outlineSettings {
      bUseBrushTransparency=false
      roundingType=HalfHeightRadius
      width=0
      color {colorUseRule=UseColor_Specified, specifiedColor=<rgba 0 0 0 0>}
      cornerRadii=<0 0 0 0>
     }
     tintColor {colorUseRule=UseColor_Specified, specifiedColor=<rgba 0.5457 0.5457 0.5457 1>}
     uVRegion {bIsValid=false, max=<0 0>, min=<0 0>}
    }
    padding=<ltrb 3 0 3 0>
   }
   slot {class=VerticalBoxSlot, properties {padding=<ltrb 0 3 0 2>}}
  }
  HorizontalBox_1 {
   widgetClassPath=&HorizontalBox
   parent=borderMain
   widget=HorizontalBox_1
   slot {class=BorderSlot, properties {padding=<ltrb 3 0 3 0>}}
  }
  btnFindCB {
   widgetClassPath=&Button
   parent=HorizontalBox_1
   widget=btnFindCB
   properties {toolTipText=Find this mesh in the Content Browser., backgroundColor=<rgba 0 0 0 0.25>}
   slot {class=HorizontalBoxSlot, properties {horizontalAlignment=HAlign_Center, verticalAlignment=VAlign_Center}}
  }
  SizeBox_0 {
   widgetClassPath=&SizeBox
   parent=btnFindCB
   widget=SizeBox_0
   properties {bOverride_HeightOverride=true, bOverride_WidthOverride=true, heightOverride=28, widthOverride=28}
   slot {class=ButtonSlot, properties {padding=<ltrb 0 0 0 0>}}
  }
  Overlay_0 {widgetClassPath=&Overlay, parent=SizeBox_0, widget=Overlay_0, slot {class=SizeBoxSlot}}
  Image_386 {
   widgetClassPath=&Image
   bIsVariable=true
   parent=Overlay_0
   widget=Image_386
   properties {
    brush {
     bIsDynamicallyLoaded=false
     drawAs=Image
     imageType=NoImage
     mirroring=NoMirror
     resourceName=None
     resourceObject=/Game/UI/Friends_UI/Social/Search
     tiling=NoTile
     imageSize=<20 20>
     margin=<ltrb 0 0 0 0>
     outlineSettings {
      bUseBrushTransparency=false
      roundingType=HalfHeightRadius
      width=0
      color {colorUseRule=UseColor_Specified, specifiedColor=<rgba 0 0 0 0>}
      cornerRadii=<0 0 0 0>
     }
     tintColor {colorUseRule=UseColor_Specified, specifiedColor=<rgba 1 1 1 1>}
     uVRegion {bIsValid=false, max=<0 0>, min=<0 0>}
    }
    colorAndOpacity=<rgba 0.3906 0.3906 0.3906 1>
   }
   slot {class=OverlaySlot, properties {horizontalAlignment=HAlign_Right}}
  }
  Image {
   widgetClassPath=&Image
   bIsVariable=true
   parent=Overlay_0
   widget=Image
   properties {
    clipping=ClipToBounds
    brush {
     bIsDynamicallyLoaded=false
     drawAs=Image
     imageType=NoImage
     mirroring=NoMirror
     resourceName=None
     resourceObject=/Game/UI/Foundation/Textures/Icons/Replay/T-Icon-Replay-Folder-L
     tiling=NoTile
     imageSize=<20 20>
     margin=<ltrb 0 0 0 0>
     outlineSettings {
      bUseBrushTransparency=false
      roundingType=HalfHeightRadius
      width=0
      color {colorUseRule=UseColor_Specified, specifiedColor=<rgba 0 0 0 0>}
      cornerRadii=<0 0 0 0>
     }
     tintColor {colorUseRule=UseColor_Specified, specifiedColor=<rgba 1 1 1 1>}
     uVRegion {bIsValid=false, max=<0 0>, min=<0 0>}
    }
    renderTransform {angle=0, scale=<1 1>, shear=<0 0>, translation=<0 3>}
   }
   slot {class=OverlaySlot, properties {verticalAlignment=VAlign_Bottom}}
  }
  btnFindWorld {
   widgetClassPath=&Button
   parent=HorizontalBox_1
   widget=btnFindWorld
   properties {toolTipText="Find all actors using this mesh in level. ", backgroundColor=<rgba 0 0 0 0.25>}
   slot {class=HorizontalBoxSlot, properties {horizontalAlignment=HAlign_Center, verticalAlignment=VAlign_Center}}
  }
  SizeBox_1 {
   widgetClassPath=&SizeBox
   parent=btnFindWorld
   widget=SizeBox_1
   properties {bOverride_HeightOverride=true, bOverride_WidthOverride=true, heightOverride=28, widthOverride=28}
   slot {class=ButtonSlot, properties {padding=<ltrb 0 0 0 0>}}
  }
  Overlay {widgetClassPath=&Overlay, parent=SizeBox_1, widget=Overlay, slot {class=SizeBoxSlot}}
  Image_1 {
   widgetClassPath=&Image
   bIsVariable=true
   parent=Overlay
   widget=Image_1
   properties {
    brush {
     bIsDynamicallyLoaded=false
     drawAs=Image
     imageType=NoImage
     mirroring=NoMirror
     resourceName=None
     resourceObject=&Search
     tiling=NoTile
     imageSize=<20 20>
     margin=<ltrb 0 0 0 0>
     outlineSettings {
      bUseBrushTransparency=false
      roundingType=HalfHeightRadius
      width=0
      color {colorUseRule=UseColor_Specified, specifiedColor=<rgba 0 0 0 0>}
      cornerRadii=<0 0 0 0>
     }
     tintColor {colorUseRule=UseColor_Specified, specifiedColor=<rgba 1 1 1 1>}
     uVRegion {bIsValid=false, max=<0 0>, min=<0 0>}
    }
    colorAndOpacity=<rgba 0.3906 0.3906 0.3906 1>
   }
   slot {class=OverlaySlot, properties {horizontalAlignment=HAlign_Right}}
  }
  Image_2 {
   widgetClassPath=&Image
   bIsVariable=true
   parent=Overlay
   widget=Image_2
   properties {
    clipping=ClipToBounds
    brush {
     bIsDynamicallyLoaded=false
     drawAs=Image
     imageType=NoImage
     mirroring=NoMirror
     resourceName=None
     resourceObject=/Engine/Tutorial/Landscape/TutorialAssets/Landscape
     tiling=NoTile
     imageSize=<20 18>
     margin=<ltrb 0 0 0 0>
     outlineSettings {
      bUseBrushTransparency=false
      roundingType=HalfHeightRadius
      width=0
      color {colorUseRule=UseColor_Specified, specifiedColor=<rgba 0 0 0 0>}
      cornerRadii=<0 0 0 0>
     }
     tintColor {colorUseRule=UseColor_Specified, specifiedColor=<rgba 1 1 1 1>}
     uVRegion {bIsValid=false, max=<0 0>, min=<0 0>}
    }
    renderTransform {angle=0, scale=<1 1>, shear=<0 0>, translation=<0 1>}
   }
   slot {class=OverlaySlot, properties {verticalAlignment=VAlign_Bottom}}
  }
  ScaleBox_0 {
   widgetClassPath=&ScaleBox
   parent=HorizontalBox_1
   widget=ScaleBox_0
   properties {clipping=ClipToBounds, stretchDirection=DownOnly}
   slot {class=HorizontalBoxSlot, properties {verticalAlignment=VAlign_Center, size {sizeRule=Fill, value=1}}}
  }
  txtMain {
   widgetClassPath=&TextBlock
   bIsVariable=true
   parent=ScaleBox_0
   widget=txtMain
   properties {
    clipping=ClipToBounds
    justification=Center
    text=Text Block
    colorAndOpacity {colorUseRule=UseColor_Specified, specifiedColor=<rgba 0.0104 0.0104 0.0104 1>}
    font {
     bForceMonospaced=false
     bMaterialIsStencil=false
     fontMaterial=null
     fontObject=&Roboto
     letterSpacing=0
     monospacedWidth=1
     size=10
     skewAmount=0
     typefaceFontName=Bold
     outlineSettings {
      bApplyOutlineToDropShadows=false
      bMiteredCorners=false
      bSeparateFillAlpha=false
      outlineMaterial=null
      outlineSize=0
      outlineColor=<rgba 0 0 0 1>
     }
    }
    shadowColorAndOpacity=<rgba 0 0 0 0.2>
   }
   slot {class=ScaleBoxSlot, properties {horizontalAlignment=HAlign_Fill}}
  }
  wrapBoxErrors {
   widgetClassPath=&WrapBox
   bIsVariable=true
   parent=HorizontalBox_1
   widget=wrapBoxErrors
   properties {innerSlotPadding=<5 5>}
   slot {class=HorizontalBoxSlot, properties {verticalAlignment=VAlign_Center, padding=<ltrb 0 2 0 2>, size {sizeRule=Fill, value=1}}}
  }
  StaticValidationError_Widget {
   widgetClassPath=&/AVT_AssetError_Widget
   bIsVariable=true
   parent=wrapBoxErrors
   widget=StaticValidationError_Widget
   slot {class=WrapBoxSlot}
  }
  btnRemove {
   widgetClassPath=&TA_Generic01_Button
   parent=HorizontalBox_1
   widget=btnRemove
   properties {
    label=" X "
    buttonColor_Default=<rgba 1 1 1 1>
    buttonStyle_Default {
     clickedSlateSound {resourceObject=null}
     disabled {
      bIsDynamicallyLoaded=false
      drawAs=Box
      imageType=NoImage
      mirroring=NoMirror
      resourceName=None
      resourceObject=/Game/Blueprints/Blutilities/ForTechArt/ArtAssets/Textures/Button02/MI_TA_ButtonBG_02_Down
      tiling=NoTile
      imageSize=<32 32>
      margin=<ltrb 0.25 0.25 0.25 0.25>
      outlineSettings {
       bUseBrushTransparency=false
       roundingType=HalfHeightRadius
       width=0
       color {colorUseRule=UseColor_Specified, specifiedColor=<rgba 0 0 0 0>}
       cornerRadii=<0 0 0 1>
      }
      tintColor {colorUseRule=UseColor_Specified, specifiedColor=<rgba 0.3333 0.3333 0.3333 1>}
      uVRegion {bIsValid=false, max=<0 0>, min=<0 0>}
     }
     disabledForeground {colorUseRule=UseColor_Foreground, specifiedColor=<rgba 1 0 1 1>}
     hovered {
      bIsDynamicallyLoaded=false
      drawAs=Box
      imageType=NoImage
      mirroring=NoMirror
      resourceName=None
      resourceObject=/Game/Blueprints/Blutilities/ForTechArt/ArtAssets/Textures/Button02/MI_TA_ButtonBG_02_Over
      tiling=NoTile
      imageSize=<32 32>
      margin=<ltrb 0.25 0.25 0.25 0.25>
      outlineSettings {
       bUseBrushTransparency=false
       roundingType=HalfHeightRadius
       width=0
       color {colorUseRule=UseColor_Specified, specifiedColor=<rgba 0 0 0 0>}
       cornerRadii=<0 0 0 1>
      }
      tintColor {colorUseRule=UseColor_Specified, specifiedColor=<rgba 0.8542 0.04 0.04 1>}
      uVRegion {bIsValid=false, max=<0 0>, min=<0 0>}
     }
     hoveredForeground {colorUseRule=UseColor_Foreground, specifiedColor=<rgba 1 0 1 1>}
     hoveredSlateSound {resourceObject=null}
     normal {
      bIsDynamicallyLoaded=false
      drawAs=Box
      imageType=NoImage
      mirroring=NoMirror
      resourceName=None
      resourceObject=/Game/Blueprints/Blutilities/ForTechArt/ArtAssets/Textures/Button02/M_TA_ButtonBG_02_Up
      tiling=NoTile
      imageSize=<32 32>
      margin=<ltrb 0.25 0.25 0.25 0.25>
      outlineSettings {
       bUseBrushTransparency=false
       roundingType=HalfHeightRadius
       width=0
       color {colorUseRule=UseColor_Specified, specifiedColor=<rgba 0 0 0 0>}
       cornerRadii=<0 0 0 1>
      }
      tintColor {colorUseRule=UseColor_Specified, specifiedColor=<rgba 0.5457 0.0199 0.0199 1>}
      uVRegion {bIsValid=false, max=<0 0>, min=<0 0>}
     }
     normalForeground {colorUseRule=UseColor_Foreground, specifiedColor=<rgba 1 0 1 1>}
     normalPadding=<ltrb 0 0 0 0>
     pressed {
      bIsDynamicallyLoaded=false
      drawAs=Box
      imageType=NoImage
      mirroring=NoMirror
      resourceName=None
      resourceObject=&MI_TA_ButtonBG_02_Down
      tiling=NoTile
      imageSize=<32 32>
      margin=<ltrb 0.25 0.25 0.25 0.25>
      outlineSettings {
       bUseBrushTransparency=false
       roundingType=HalfHeightRadius
       width=0
       color {colorUseRule=UseColor_Specified, specifiedColor=<rgba 0 0 0 0>}
       cornerRadii=<0 0 0 1>
      }
      tintColor {colorUseRule=UseColor_Specified, specifiedColor=<rgba 0.2396 0.0087 0.0087 1>}
      uVRegion {bIsValid=false, max=<0 0>, min=<0 0>}
     }
     pressedForeground {colorUseRule=UseColor_Foreground, specifiedColor=<rgba 1 0 1 1>}
     pressedPadding=<ltrb 0 0 0 0>
     pressedSlateSound {resourceObject=null}
    }
    labelColor_Default=<rgba 0.9531 0.9531 0.9531 1>
   }
   slot {class=HorizontalBoxSlot, properties {verticalAlignment=VAlign_Center}}
  }
  imageEdge {
   widgetClassPath=&Image
   bIsVariable=true
   parent=VerticalBox_0
   widget=imageEdge
   properties {
    brush {
     bIsDynamicallyLoaded=false
     drawAs=Image
     imageType=NoImage
     mirroring=NoMirror
     resourceName=None
     resourceObject=null
     tiling=NoTile
     imageSize=<32 2>
     margin=<ltrb 0 0 0 0>
     outlineSettings {
      bUseBrushTransparency=false
      roundingType=HalfHeightRadius
      width=0
      color {colorUseRule=UseColor_Specified, specifiedColor=<rgba 0 0 0 0>}
      cornerRadii=<0 0 0 0>
     }
     tintColor {colorUseRule=UseColor_Specified, specifiedColor=<rgba 1 1 1 1>}
     uVRegion {bIsValid=false, max=<0 0>, min=<0 0>}
    }
    colorAndOpacity=<rgba 0.8123 0.8217 0.8385 0.75>
   }
   slot {class=VerticalBoxSlot}
  }
  AVT_ListItemEditorBridge_Widget {
   widgetClassPath=&/AVT_ListItemEditorBridge_Widget
   bIsVariable=true
   parent=Overlay_7
   widget=AVT_ListItemEditorBridge_Widget
   slot {class=OverlaySlot}
  }
 ]
}
